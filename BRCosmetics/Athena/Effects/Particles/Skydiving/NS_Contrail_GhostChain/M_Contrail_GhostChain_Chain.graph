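# Lit Niagara mesh-particle material for the chains in the GhostChain skydiving contrail, reusing Embers backpack color and normal textures.
# An emissive mask is tinted by particle color; the Fallback static switch also multiplies in a curve-atlas color sampled along UV1.
# The same switch enables sine and cosine swing offsets, driven by dynamic parameters and scaled along the chain by a strength curve.
# Inverted red vertex color plus a Cut parameter forms a cutoff value.
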Material /BRCosmetics/Athena/Effects/Particles/Skydiving/NS_Contrail_GhostChain/M_Contrail_GhostChain_Chain
graph {
 properties {bUsedWithNiagaraMeshParticles=true, pixelDepthOffsetMode=PDOM_Legacy, translucencyLightingMode=TLM_Surface, usedShadingModels=Default Lit}
 nodes [
  n1 = DynamicParameter() {defaultValue=<rgba 0 0 0 0>, paramNames [X,Y,Z,Cut]}
  Add(A=OneMinus(None=VertexColor().R), B=@n1.Cut)
  Constant3Vector()
  n2 = Add(A=TextureSample() {texture=&/T_GhostChain_EmissiveMask}.RGB) {constB=0.15}
  n3 = MaterialFunctionCall(Float2=TextureCoordinate() {coordinateIndex=1}) {materialFunction=&BreakOutFloat2Components}
  n4 = CurveAtlasRowParameter(CurveTime=@n3.G) {
   atlas=&/ChainFallBackAtlas
   curve=&/ChainFallBackAtlas_ChainColor
   defaultValue=1
   parameterName=FallbackColor
  }
  n5 = ParticleColor()
  n6 = Multiply(A=@n4, B=@n5.RGB)
  MaterialFunctionCall(Emissive=Multiply(A=Saturate(None=@n2), B=StaticSwitchParameter(True=@n6, False=@n5.RGB) {parameterName=Fallback})) {materialFunction=&MF_FX_EmissiveAdjustments}
  Power(
   Base=TextureSample() {texture=/BRCosmetics/Accessories/FORT_Backpacks/Backpack_M_MED_Embers/Textures/T_M_MED_Embers_Backpack_D}.RGB
   Exp=ScalarParameter() {defaultValue=0.9, parameterName=BasecolorPow}
  )
  ScalarParameter() {defaultValue=1, parameterName=Metallic}
  ScalarParameter() {defaultValue=0.55, parameterName=Roughness}
  ScalarParameter() {defaultValue=0.1, parameterName=Specular}
  n7 = OneMinus(None=@n3.G)
  n8 = CurveAtlasRowParameter(CurveTime=@n7) {
   atlas=&ChainFallBackAtlas
   curve=&/ChainFallBackAtlas_WPOStr
   parameterName=WPOStrCurve
  }
  n9 = DynamicParameter() {parameterIndex=2, defaultValue=<rgba 1.5 1 1 1>, paramNames [SwingTimerX, SwingFrequencyX, SwingStrengthX, Param4]}
  n10 = Subtract(A=Multiply(A=@n7, B=@n9.SwingFrequencyX), B=Fmod(A=@n9.SwingTimerX, B=Constant() {r=1}))
  n11 = Multiply(
   A=Multiply(
    A=Sine(None=@n10)
    B=MaterialFunctionCall(X=@n1.X, Y=@n1.X, Z=@n1.X) {materialFunction=&MakeFloat3}.Result
   )
   B=@n9.SwingStrengthX
  )
  n12 = DynamicParameter() {parameterIndex=1, defaultValue=<rgba 0 0 0 0>, paramNames [X,Y,Z,Param4]}
  n13 = DynamicParameter() {parameterIndex=3, defaultValue=<rgba 1.5 1 1 1>, paramNames [SwingTimerY, SwingFrequencyY, SwingStrengthY, Param4]}
  n14 = Subtract(A=Multiply(A=@n7, B=@n13.SwingFrequencyY), B=Fmod(A=@n13.SwingTimerY, B=Constant() {r=1}))
  n15 = Multiply(
   A=Multiply(
    A=MaterialFunctionCall(X=@n12.X, Y=@n12.X, Z=@n12.X) {materialFunction=&MakeFloat3}.Result
    B=Cosine(None=@n14)
   )
   B=@n13.SwingStrengthY
  )
  StaticSwitchParameter(True=Multiply(A=@n8.R, B=Add(A=@n11, B=@n15)) {constA=1}, False=Constant()) {parameterName=Fallback}
  TextureSample() {samplerType=SAMPLERTYPE_Normal, texture=/BRCosmetics/Accessories/FORT_Backpacks/Backpack_M_MED_Embers/Textures/T_M_MED_Embers_Backpack_N}
 ]
}
