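# Niagara skydiving contrail cosmetic (GhostChain) with green flame trails, sparks, smoke and lights following the player's skeletal bones.
# System Dive values scale spawn rates, velocities, mesh size and brightness, while system movement orients the effect.
# Two persistent FireTrail meshes follow bone pairs, passing bend vectors to materials; sprite emitters split particles between two bones.
# User flags adjust sizes, velocities and colors for front-end preview.
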NiagaraSystem /BRCosmetics/Athena/Effects/Particles/Skydiving/NS_Contrail_GhostChain/NS_Contrails_GhostChain_V1d9d1
dependencies {
 usedDataInterfaces [
  &NiagaraDataInterfaceCurve
  &NiagaraDataInterfaceCurlNoise
  &NiagaraDataInterfaceVectorCurve
  &NiagaraDataInterfaceColorCurve
  &NiagaraDataInterfaceArrayFloat3
  &NiagaraDataInterfaceArrayInt32
  &NiagaraDataInterfaceArrayFloat
  &NiagaraDataInterfaceVector2DCurve
 ]
 usedDynamicInputs [&SelectVectorFromArray, &TransformVector]
 usedModules [
  &self:SystemSpawnScript.SystemScriptSource.SystemScriptGraph.NiagaraNodeAssignment_0.SetVariables_#1
  &self:SystemSpawnScript.SystemScriptSource.SystemScriptGraph.NiagaraNodeAssignment_8.SetVariables_#2
  &self:SystemSpawnScript.SystemScriptSource.SystemScriptGraph.NiagaraNodeAssignment_3.SetVariables_#3
  &self:FakeDivePreset
  &SystemState
  /Niagara/Modules/System/SystemLifeCycle
  /Game/Effects/Niagara/Modules/Skydiving/FN_DivingAttributes/v4/FN_DivingAttributes_v4
  &self:SystemSpawnScript.SystemScriptSource.SystemScriptGraph.NiagaraNodeAssignment_4.SetVariables_#4
  &self:FakeDive0
  &self:SystemSpawnScript.SystemScriptSource.SystemScriptGraph.NiagaraNodeAssignment_6.SetVariables_#5
  &self:SystemSpawnScript.SystemScriptSource.SystemScriptGraph.NiagaraNodeAssignment_7.SetVariables_#6
  &self:SystemSpawnScript.SystemScriptSource.SystemScriptGraph.NiagaraNodeAssignment_1.SetVariables_#7
  &self:SystemSpawnScript.SystemScriptSource.SystemScriptGraph.NiagaraNodeAssignment_2.SetVariables_#8
  &self:ADive
  &self:SystemSpawnScript.SystemScriptSource.SystemScriptGraph.NiagaraNodeAssignment_5.SetVariables_#9
  &self:SystemSpawnScript.SystemScriptSource.SystemScriptGraph.NiagaraNodeAssignment_9.SetVariables_#10
 ]
}
emitters {
 Empty {
  emitterUpdateScript [
   &EmitterState(
    Enable Distance Culling=true
    Life Cycle Mode=::NewEnumerator1
    Loop Duration=1
    MaxDistance=2000
    Module.bUseMaxDistance=true
    Module.bUseMinDistance=false
    Scalability Mode=::NewEnumerator1
    UseLoopDelay=false
   )
   &SpawnRate(
    SpawnRate=&Lerp_Float(
     A=0
     B=60
     Alpha=&FloatFromCurve(
      FloatCurve=DataInterface() {propertyValues {Curve {keys [{interpMode=RCIM_Cubic, leaveTangent=4.6729, tangentMode=RCTM_Break, time=0, value=0}, <at 0.7 1>]}}}
      CurveIndex=$System.Dive
     )
    )
    Use Spawn Probability=false
   )
  ]
  particleSpawnScript [
   &InitializeParticle(
    Color Mode=::NewEnumerator0
    Lifetime Max=0.75
    Lifetime Min=0.3
    Lifetime Mode=::NewEnumerator1
    UsePositionOffset=false
    Write Material Random=true
   )
   &SphereLocation(
    MaskSpawnGroup=false
    Override Seed=false
    Sphere Radius=10
    Surface Only=false
    UseNonUniformScale=false
    UseOffset=false
    UseSphereOrientation=false
   )
   &AddVelocityInCone(
    Cone Angle=15
    Cone Axis=/Niagara/DynamicInputs/Helpers/SafeNormalizeVector(Vector To Normalize:Vector3f=$System.DeltaSystemPosition, Fallback Vector:Vector3f=<0 0 1>)
    Module.Use Velocity Falloff On Cone Axis=false
    Velocity Strength=&Multiply_Float(
     A=&RandomRangeFloat(Minimum=1094.1201, Maximum=1534.72)
     B=&FloatFromCurve(
      FloatCurve=DataInterface() {propertyValues {Curve {keys [
          {arriveTangent=0.4749, interpMode=RCIM_Cubic, leaveTangent=0.4749, tangentMode=RCTM_User, time=0, value=0.3429}
          {interpMode=RCIM_Cubic, tangentMode=RCTM_User, time=0.7, value=1}
         ]}}}
      CurveIndex=$System.Dive
     )
    )
   )
   /SetVariables(Particles.SpriteSize=&RandomRangeVector2D(Minimum=<0.7 2>, Maximum=<1.7 2>)) {isSetParameters=true}
  ]
  particleUpdateScript [
   &Drag(Drag=20, Use Linear Drag=true, Use Rotational Drag=false)
   &SolveForcesAndVelocity(Clamp Velocity=false, Limit Acceleration=false)
   &ParticleState()
   /Game/Effects/Niagara/Modules/SkeletalMesh/FN_SkeletalBoneData(Index:NiagaraInt32=&MakeCustomIntFromBool(
     Bool=&SetBoolByIntComparison(
      A=&ModuloInt(A=&self:UniqueIDInt(), B=2)
      B=1
      Comparison Type=::NewEnumerator4
     )
     True INT=1
     False INT=0
    )) {module=FN_SkeletalBoneData001}
   &self:UnpakBonDataPosition(Input 1:ENiagaraBoneData=$Particles.FN_SkeletalBoneData001.BoneData) {module=UnpakBonDataPosition}
   /SetVariables(Particles.FinalPosition:Vector3f=&Add_Vector(A=Linked() {linkedVariable {name=Particles.Position, type=NiagaraPosition}}, B=$Particles.UnpakBonDataPosition.BoneDataPosition)) {isSetParameters=true}
   &ScaleSpriteSizeBySpeed(Max Scale Factor=<1 6.834>)
   &Color(Color=&MultiplyLinearColor(
     Linear Color=&ColorFromCurve(ColorCurve=DataInterface() {propertyValues {
        AlphaCurve {keys [
          {arriveTangent=-12.6265, interpMode=RCIM_Cubic, leaveTangent=-12.6265, tangentMode=RCTM_User, time=0.1372, value=1}
          {arriveTangent=-1.1591, interpMode=RCIM_Cubic, leaveTangent=-1.1591, time=0.2193, value=0.3941}
          <at 1 0 RCIM_Cubic>
         ]}
        BlueCurve {keys [<at 0 0.1 RCIM_Cubic>, {arriveTangent=-0.1, interpMode=RCIM_Cubic, leaveTangent=-0.1, time=0.1519, value=0.0667}, <at 1 0>]}
        GreenCurve {keys [<at 0 2 RCIM_Cubic>, {arriveTangent=-2, interpMode=RCIM_Cubic, leaveTangent=-2, time=0.1519, value=1}, <at 1 0>]}
        RedCurve {keys [<at 0 0.765 RCIM_Cubic>, {arriveTangent=-0.565, interpMode=RCIM_Cubic, leaveTangent=-0.565, time=0.1519, value=0}, <at 1 0.2>]}
       }})
     Scale Factor=/Niagara/DynamicInputs/TypeConversions/MakeCustomVector4FromBool(
      Bool:NiagaraBool=$User.bIsFrontEndPreview
      True Vector4:Vector4f=<1 1 1 1>
      False Vector4:Vector4f=&MakeFloat4FromVectorAndFloat(
       "Vector (RGB)"=&VectorFromFloat(Value=&FloatFromCurve(
         FloatCurve=DataInterface() {propertyValues {Curve {keys [
             {interpMode=RCIM_Cubic, tangentMode=RCTM_User, time=0, value=1}
             {arriveTangent=4.3422, interpMode=RCIM_Cubic, leaveTangent=4.3422, tangentMode=RCTM_User, time=1, value=2.5}
            ]}}}
         CurveIndex=&RandomRangeFloat(Minimum=0, Maximum=1)
        ))
       "Float (Alpha)"=1
      )
     )
    ))
  ]
  rendererClasses [&NiagaraSpriteRendererProperties]
  renderers [{rendererClass=&NiagaraSpriteRendererProperties}]
 }
 Empty001 {
  emitterUpdateScript [
   &EmitterState(
    Enable Distance Culling=true
    Life Cycle Mode=::NewEnumerator1
    Loop Duration=1
    MaxDistance=2000
    Module.bUseMaxDistance=true
    Module.bUseMinDistance=false
    Scalability Mode=::NewEnumerator1
    UseLoopDelay=false
   )
   &SpawnRate(
    SpawnRate=&Lerp_Float(
     A=0
     B=15
     Alpha=&FloatFromCurve(
      FloatCurve=DataInterface() {propertyValues {Curve {keys [{tangentMode=RCTM_User, time=0, value=1}, <at 0.1 1>, {tangentMode=RCTM_User, time=0.11, value=0}]}}}
      CurveIndex=$System.Dive
     )
    )
    Use Spawn Probability=false
   )
  ]
  particleSpawnScript [
   &InitializeParticle(
    Color Mode=::NewEnumerator0
    Lifetime Max=0.5
    Lifetime Min=0.35
    Lifetime Mode=::NewEnumerator1
    UsePositionOffset=false
    Write Material Random=true
   )
   &AddVelocityInCone(
    Cone Angle=30
    Cone Axis=&FN_FrontEndVector3Modifier(
     Gameplay=&SafeNormalizeVector(Vector To Normalize:Vector3f=$System.DeltaSystemPosition, Fallback Vector:Vector3f=<0 0 1>)
     Front End=<0 -1 1>
    )
    Module.Use Velocity Falloff On Cone Axis=false
    Velocity Strength=&FN_FrontEndFloatModifier(Gameplay=300, Front End=150)
   )
   /Niagara/Modules/Spawn/Location/SphereLocation(MaskSpawnGroup=false, Module.Override Randomness=false, Sphere Radius=10, Surface Only=false)
  ]
  particleUpdateScript [
   /SetVariables(Particles.Drag:NiagaraFloat=&RandomRangeFloat(Minimum=4, Maximum=6)) {isSetParameters=true}
   &VectorNoiseForce(Force Amount=2500, Module.Use Random Noise Time Interval=true, Random Noise Time Interval=0.2)
   &Drag(Drag=$Particles.Drag, Use Linear Drag=true, Use Rotational Drag=false)
   &SolveForcesAndVelocity(Clamp Velocity=false, Limit Acceleration=false)
   &ParticleState()
   &FN_SkeletalBoneData(Index:NiagaraInt32=&MakeCustomIntFromBool(
     Bool=&SetBoolByIntComparison(
      A=&ModuloInt(A=&self:UniqueIDInt(), B=2)
      B=1
      Comparison Type=::NewEnumerator4
     )
     True INT=1
     False INT=0
    )) {module=FN_SkeletalBoneData001}
   &self:UnpakBonDataPosition(Input 1:ENiagaraBoneData=$Particles.FN_SkeletalBoneData001.BoneData) {module=UnpakBonDataPosition}
   /SetVariables(Particles.FinalPosition:Vector3f=&Add_Vector(A=Linked() {linkedVariable {name=Particles.Position, type=NiagaraPosition}}, B=$Particles.UnpakBonDataPosition.BoneDataPosition)) {isSetParameters=true}
   /SetVariables(Particles.SpriteSize=&UniformRangedVector2D(
     Minimum=&FN_FrontEndVector2Modifier(Front End:Vector2f=<2 5>, Gameplay:Vector2f=<1.5 5>)
     Maximum=&FN_FrontEndVector2Modifier(Front End:Vector2f=<1 1>, Gameplay:Vector2f=<0.7 1>)
    )) {isSetParameters=true}
   &Color(Color=&MultiplyLinearColor(
     Linear Color=&ColorFromCurve(ColorCurve=DataInterface() {propertyValues {
        AlphaCurve {keys [<at 0 1 RCIM_Cubic>,<at 1 0 RCIM_Cubic>]}
        BlueCurve {keys [<at 0 0 RCIM_Cubic>, {interpMode=RCIM_Cubic, leaveTangent=0.0235, time=0.2343, value=0}]}
        GreenCurve {keys [<at 0 2 RCIM_Cubic>, {interpMode=RCIM_Cubic, leaveTangent=-1.4085, time=0.2343, value=1}]}
        RedCurve {keys [<at 0 0.6 RCIM_Cubic>, {interpMode=RCIM_Cubic, leaveTangent=-0.8451, time=0.2343, value=0.05}]}
       }})
     Scale Factor=&MakeFloat4FromVectorAndFloat(
      "Vector (RGB)"=&VectorFromFloat(Value=&FloatFromCurve(
        FloatCurve=DataInterface() {propertyValues {Curve {keys [
            {interpMode=RCIM_Cubic, tangentMode=RCTM_User, time=0, value=1}
            {arriveTangent=9, interpMode=RCIM_Cubic, leaveTangent=9, time=0.676, value=3.3562}
            {interpMode=RCIM_Cubic, tangentMode=RCTM_User, time=1, value=20}
           ]}}}
        CurveIndex=&RandomRangeFloat(Minimum=0, Maximum=1)
       ))
      "Float (Alpha)"=1
     )
    ))
  ]
  rendererClasses [&NiagaraSpriteRendererProperties]
  renderers [{rendererClass=&NiagaraSpriteRendererProperties}]
 }
 FireTrail {
  emitterUpdateScript [
   &EmitterState(
    Enable Distance Culling=true
    Inactive Response=::NewEnumerator1
    Life Cycle Mode=::NewEnumerator1
    Loop Behavior=::NewEnumerator1
    Loop Duration Mode=::NewEnumerator1
    MaxDistance=10000
    Module.bUseMaxDistance=true
    Module.bUseMinDistance=false
    Scalability Mode=::NewEnumerator1
    UseLoopDelay=false
   )
   &SpawnBurst_Instantaneous(Spawn Count=2, Use Spawn Probability=false)
   /SetVariables(Emitter.Randomer:NiagaraFloat=&RandomRangeFloat(Minimum=0, Maximum=1, Evaluation Type=::NewEnumerator1)) {isSetParameters=true}
  ]
  particleSpawnScript [
   &FN_SetSpawnGroup(Number of Spawn Groups=2)
   &InitializeParticle(Color Mode=::NewEnumerator0, Position Mode=::NewEnumerator0, Write Material Random=true)
  ]
  particleUpdateScript [
   &ParticleState(Kill Particles When Lifetime Has Elapsed=false, Let Infinitely Lived Particles Die When Emitter Deactivates=true)
   &FN_SkeletalBoneData(Index:NiagaraInt32=&SelectIntFromArray(Direct Array Index=$Particles.SpawnGroup, Int Selection Array=DataInterface() {propertyValues {IntData [0,1]}}))
   &FN_SkeletalBoneData(Index:NiagaraInt32=&SelectIntFromArray(Direct Array Index=$Particles.SpawnGroup, Int Selection Array=DataInterface() {propertyValues {IntData [5,6]}})) {module=FN_SkeletalBoneData001}
   /SetVariables(Particles.Position=&Lerp_Vector(
     A=/BRCosmetics/Athena/DrivableVehicles/Meatball/Meatball_Large/FX/DynamicInputs/FN_GetBoneDataPosition(BoneData:ENiagaraBoneData=$Particles.FN_SkeletalBoneData.BoneData, Space:ENiagaraCoordinateSpace=::Local)
     B=&FN_GetBoneDataPosition(BoneData:ENiagaraBoneData=$Particles.FN_SkeletalBoneData001.BoneData, Space:ENiagaraCoordinateSpace=::Local)
     Alpha=0.2
    )) {isSetParameters=true}
   /SetVariables(
    Particles.Scale=&MakeCustomVectorFromBool(Bool=$User.bIsFrontEndPreview, True Vector=<0.73 0.73 1>, False Vector=<0.8 0.8 1>)
    Particles.Scale2:Vector3f=&MakeCustomVectorFromBool(Bool=$User.bIsFrontEndPreview, True Vector=<0.35 0.35 0.4>, False Vector=<0.45 0.45 0.4>)
   ) {isSetParameters=true}
   /SetVariables(
    Particles.Scale=&Multiply_Vector(
     A=$Particles.Scale
     B=&VectorFromCurve(
      VectorCurve=DataInterface() {propertyValues {
        XCurve {keys [{tangentMode=RCTM_User, time=0, value=0}, <at 0.03 0.85>, {tangentMode=RCTM_User, time=0.7, value=1}]}
        YCurve {keys [{tangentMode=RCTM_User, time=0, value=0}, <at 0.03 0.85>, {tangentMode=RCTM_User, time=0.7, value=1}]}
        ZCurve {keys [{tangentMode=RCTM_User, time=0, value=0}, <at 0.03 0.45>, {tangentMode=RCTM_User, time=0.7, value=1}]}
       }}
      CurveIndex=$System.Dive
     )
    )
    Particles.Scale2:Vector3f=&Multiply_Vector(
     A=$Particles.Scale2
     B=&VectorFromCurve(
      VectorCurve=DataInterface() {propertyValues {
        XCurve {keys [{tangentMode=RCTM_User, time=0, value=0}, <at 0.03 0.85>, {tangentMode=RCTM_User, time=0.7, value=1}]}
        YCurve {keys [{tangentMode=RCTM_User, time=0, value=0}, <at 0.03 0.85>, {tangentMode=RCTM_User, time=0.7, value=1}]}
        ZCurve {keys [{tangentMode=RCTM_User, time=0, value=0}, <at 0.03 0.65>, {tangentMode=RCTM_User, time=0.7, value=1}]}
       }}
      CurveIndex=$System.Dive
     )
    )
   ) {isSetParameters=true}
   /SetVariables(
    Particles.Scale=&Multiply_Vector(
     A=$Particles.Scale
     B=&MakeVector(
      X=1
      Y=1
      Z=&FloatFromCurve(
       FloatCurve=DataInterface() {propertyValues {Curve {keys [
           {arriveTangent=2.4928, interpMode=RCIM_Cubic, leaveTangent=2.4928, tangentMode=RCTM_User, time=0, value=0}
           {arriveTangent=0.9995, interpMode=RCIM_Cubic, leaveTangent=0.9995, tangentMode=RCTM_User, time=0.3675, value=0.6968}
           {arriveTangent=-0.0405, interpMode=RCIM_Cubic, leaveTangent=-0.0405, tangentMode=RCTM_User, time=1, value=1}
          ]}}}
       CurveIndex=$System.ADive.ADive
      )
     )
    )
    Particles.Scale2:Vector3f=&Multiply_Vector(
     A=$Particles.Scale2
     B=&MakeVector(
      X=1
      Y=1
      Z=&FloatFromCurve(
       FloatCurve=DataInterface() {propertyValues {Curve {keys [
           {arriveTangent=4.1785, interpMode=RCIM_Cubic, leaveTangent=4.1785, tangentMode=RCTM_User, time=0, value=0}
           {arriveTangent=0.9995, interpMode=RCIM_Cubic, leaveTangent=0.9995, tangentMode=RCTM_User, time=0.178, value=0.6716}
           {arriveTangent=-0.0405, interpMode=RCIM_Cubic, leaveTangent=-0.0405, tangentMode=RCTM_User, time=1, value=1}
          ]}}}
       CurveIndex=$System.ADive.ADive
      )
     )
    )
   ) {isSetParameters=true}
   /SetVariables(Particles.MeshOrientation=&MultiplyQuaternion(
     Quaternion A=&FindQuatBetweenTwoVectors(From Vector=<0 0 1>, To Vector=&NormalizeVector(Vector To Normalize=$System.DeltaSystemPosition))
     Quaternion B=&FindQuatBetweenTwoVectors(
      From Vector=<0 0 1>
      To Vector=&NormalizeVector(Vector To Normalize=&MakeVector(
        X=&Multiply_Float(
         A=&SelectFloatFromArray(Direct Array Index=&SpawnGroup(), Float Selection Array=DataInterface() {propertyValues {FloatData [-0.4,0.4]}})
         B=&FloatFromCurve(
          FloatCurve=DataInterface() {propertyValues {Curve {keys [
              {interpMode=RCIM_Cubic, tangentMode=RCTM_Break, time=0, value=1}
              {arriveTangent=-2.2148, interpMode=RCIM_Cubic, tangentMode=RCTM_Break, time=1, value=0.2}
             ]}}}
          CurveIndex=$System.Dive
         )
        )
        Y=0
        Z=1
       ))
     )
    )) {isSetParameters=true}
   &self:MeshRotateRate(Axis:Vector3f=<0 0 1>, Rate:NiagaraFloat=&MaskFloatBySpawnGroup(Masked Float=-5, Passthrough Float=5)) {module=MeshRotateRate}
   /SetVariables(Particles.MeshOrientation2:Quat4f=&MultiplyQuaternion(Quaternion A=$Particles.MeshOrientation, Quaternion B=$Particles.MeshRotateRate.RotateQuat)) {isSetParameters=true}
   &self:Timer(ScaleDeltaTime:NiagaraFloat=&Lerp_Float(A=3, B=4.5, Alpha=$System.Dive)) {module=Timer}
   &self:CalcBendVec(DeltaSystemPosition:Vector3f=$System.DeltaSystemPosition) {module=CalcBendVec}
   &DynamicMaterialParameters(
    Index 0 Param 1=&MakeFloatFromVector(Vector=$Particles.BendVec, Channel=::NewEnumerator0)
    Index 0 Param 2=&MakeFloatFromVector(Vector=$Particles.BendVec, Channel=::NewEnumerator1)
    Index 0 Param 3=&MakeFloatFromVector(Vector=$Particles.BendVec, Channel=::NewEnumerator2)
    Index 0 Param 4=5
    Index 3 Param 1=$Particles.Timer.Timer
    Module.Param0WriteEnabled=true
    Module.Param0WriteEnabled3=true
    Module.Param1WriteEnabled=true
    Module.Param1WriteEnabled3=true
    Module.Param2WriteEnabled=true
    Module.Param2WriteEnabled3=true
    Module.Param3WriteEnabled=true
    Module.Param3WriteEnabled3=true
    Write Parameter Index 3=true
   )
   /SetVariables(Particles.LightColor=&MultiplyLinearColor(
     Linear Color=<rgba 0 10000 5427.8208 1>
     Scale Factor=&MakeVector4_FromFloat(Float=&FloatFromCurve(
       FloatCurve=DataInterface() {propertyValues {Curve {keys [
           {arriveTangent=-0.9091, interpMode=RCIM_Cubic, time=0, value=0.3}
           {arriveTangent=1.2822, interpMode=RCIM_Cubic, leaveTangent=1.2822, tangentMode=RCTM_User, time=0.5836, value=0.608}
           <at 1 1 RCIM_Cubic>
          ]}}}
       CurveIndex=$Emitter.Randomer
      ))
    )) {isSetParameters=true}
   &ScaleColor(
    Scale RGB=&VectorFromFloat(Value=&FloatFromCurve(
      FloatCurve=DataInterface() {propertyValues {Curve {keys [
          {arriveTangent=1.5795, interpMode=RCIM_Cubic, leaveTangent=1.5795, tangentMode=RCTM_User, time=0.2, value=0.6}
          {arriveTangent=0.8, interpMode=RCIM_Cubic, leaveTangent=0.8, time=0.4732, value=0.9278}
          {interpMode=RCIM_Cubic, tangentMode=RCTM_User, time=0.7, value=1}
         ]}}}
      CurveIndex=$System.Dive
     ))
    ScaleA=true
    ScaleRGB=true
   ) {enabled=false, module=ScaleColor001}
   /SetVariables(Particles.LightColor=&MultiplyLinearColor(
     Linear Color=$Particles.LightColor
     Scale Factor=&MakeVector4_FromFloat(Float=&FloatFromCurve(
       FloatCurve=DataInterface() {propertyValues {Curve {keys [
           {arriveTangent=1.5795, interpMode=RCIM_Cubic, time=0.2, value=0}
           {arriveTangent=1.8182, interpMode=RCIM_Cubic, leaveTangent=1.8182, time=0.4842, value=0.8024}
           <at 0.75 1 RCIM_Cubic>
          ]}}}
       CurveIndex=$System.Dive
      ))
    )) {isSetParameters=true}
  ]
  rendererClasses [&NiagaraMeshRendererProperties, &NiagaraLightRendererProperties]
  renderers [
   {rendererClass=&NiagaraMeshRendererProperties, rendererIndex=0}
   {rendererClass=&NiagaraMeshRendererProperties, rendererIndex=1}
   {rendererClass=&NiagaraLightRendererProperties, rendererIndex=2}
  ]
 }
 Flames-Local {
  emitterUpdateScript [
   &EmitterState(
    Enable Distance Culling=true
    Life Cycle Mode=::NewEnumerator1
    Loop Delay=1
    Loop Duration=1
    MaxDistance=3000
    Module.bUseMaxDistance=true
    Module.bUseMinDistance=false
    Scalability Mode=::NewEnumerator1
    UseLoopDelay=false
   )
   &SpawnRate(SpawnRate=&Lerp_Float(A=0, B=30, Alpha=$System.Dive), Use Spawn Probability=false)
  ]
  particleSpawnScript [
   /Niagara/Modules/Spawn/Initialization/InitializeParticle(
    Color=<rgba 5 3.0056 1.6206 0.75>
    Lifetime=&UniformRangedFloat(Minimum=0.25, Maximum=0.35)
    Module.Write Color=false
    Module.Write Lifetime=true
    Module.Write Mass=false
    Module.Write Material Random=true
    Module.Write Position=false
    Module.Write Scale=false
    Module.Write SpriteRotation=true
    Module.Write SpriteSize=true
    Position=&SimulationPosition()
    Sprite Rotation=&UniformRangedFloat(Minimum=0, Maximum=360)
    Sprite Size=&MakeCustomVector2DFromBool(
     Bool=&RandomBool()
     True Vector2D=&Vector2DFromFloat(Value=&RandomRangeFloat(Minimum=30, Maximum=35))
     False Vector2D=&Vector2DFromFloat(Value=&RandomRangeFloat(Minimum=-30, Maximum=-35))
    )
   )
   &Color(Color=<rgba 3 3 3 1>)
   &DynamicMaterialParameters(
    Index 0 Param 2=&UniformRangedFloat(Minimum=0, Maximum=1)
    Index 0 Param 3=1
    Index 0 Param 4=&UniformRangedFloat(Minimum=1, Maximum=2)
    Module.Param0WriteEnabled=false
    Module.Param1WriteEnabled=true
    Module.Param2WriteEnabled=false
    Module.Param3WriteEnabled=true
   ) {module=DynamicMaterialParameters002}
   &CameraOffset(Camera Offset Amount=&RandomRangeFloat(Minimum=-3, Maximum=8))
   /Niagara/Modules/Spawn/Location/V2/SphereLocation(
    MaskSpawnGroup=false
    Override Seed=false
    Sphere Radius=10
    Surface Only=false
    UseNonUniformScale=false
    UseOffset=false
    UseSphereOrientation=false
   )
   &AddVelocity(Velocity=&Multiply_Vector(
     A=$System.DeltaSystemPosition
     B=&Multiply_Vector(
      A=&VectorFromFloat(Value=&RandomRangeFloat(Minimum=1.3, Maximum=2.5))
      B=&MakeCustomVectorFromBool(Bool=$User.bIsFrontEndPreview, True Vector=&VectorFromFloat(Value=170), False Vector=<1 1 1>)
     )
    ))
  ]
  particleUpdateScript [
   &SolveForcesAndVelocity(Clamp Velocity=false, Limit Acceleration=false)
   &ParticleState()
   &ScaleColor(
    Scale Mode=::NewEnumerator1
    Scale RGBA=&Vector4FromCurve(Vector4Curve=DataInterface() {propertyValues {
       AlphaCurve {keys [<at 0.0024 0>,<at 0.3018 1.2>,<at 0.725 0.2>,<at 0.9307 0>]}
       BlueCurve {keys [<at 0.15 0>,<at 0.5 0.016>,<at 0.85 0.3333>]}
       GreenCurve {keys [<at 0.15 1>,<at 0.5 1>,<at 0.85 1>]}
       RedCurve {keys [<at 0.15 0.3333>,<at 0.5 0.016>,<at 0.85 0>]}
      }})
    ScaleRGBA=true
   )
   /SetVariables(Particles.SubImageIndex=&FloatFromCurve(FloatCurve=DataInterface() {propertyValues {Curve {keys [{interpMode=RCIM_Cubic, leaveTangent=69.512, tangentMode=RCTM_Break, time=0, value=0}, <at 1 34>]}}})) {isSetParameters=true}
   &ScaleSpriteSize(Scale Factor=&Vector2DFromFloat(Value=&FloatFromCurve(FloatCurve=DataInterface() {propertyValues {Curve {keys [
          {arriveTangent=7.5764, interpMode=RCIM_Cubic, leaveTangent=7.5765, tangentMode=RCTM_User, time=0, value=0.02}
          {arriveTangent=-0.0353, interpMode=RCIM_Cubic, leaveTangent=-0.0353, tangentMode=RCTM_User, time=0.2, value=1.2}
          <at 1 0.25 RCIM_Cubic>
         ]}}})))
   &DynamicMaterialParameters(
    Index 0 Param 1=&FloatFromCurve(FloatCurve=DataInterface() {propertyValues {Curve {keys [
         {interpMode=RCIM_Cubic, leaveTangent=5.5932, tangentMode=RCTM_Break, time=0, value=0}
         {arriveTangent=1.2717, interpMode=RCIM_Cubic, leaveTangent=1.2717, tangentMode=RCTM_User, time=0.2281, value=0.7279}
         {tangentMode=RCTM_User, time=1, value=1}
        ]}}})
    Index 0 Param 2=&RandomRangeFloat(Minimum=0.33, Maximum=1)
    Index 0 Param 3=&FloatFromCurve(FloatCurve=DataInterface() {propertyValues {Curve {keys [<at 0.8 1>,<at 1 2>]}}})
    Index 0 Param 4=1
    Module.Param0WriteEnabled=true
    Module.Param1WriteEnabled=true
    Module.Param2WriteEnabled=true
    Module.Param3WriteEnabled=true
   ) {module=DynamicMaterialParameters001}
   &FN_SkeletalBoneData(Index:NiagaraInt32=&MakeCustomIntFromBool(
     Bool=&SetBoolByIntComparison(
      A=&ModuloInt(A=&self:UniqueIDInt(), B=2)
      B=1
      Comparison Type=::NewEnumerator4
     )
     True INT=1
     False INT=0
    ))
   &self:UnpakBonDataPosition(Input 1:ENiagaraBoneData=$Particles.FN_SkeletalBoneData.BoneData) {module=UnpakBonDataPosition}
   /SetVariables(Particles.FinalPosition:Vector3f=&Add_Vector(A=Linked() {linkedVariable {name=Particles.Position, type=NiagaraPosition}}, B=$Particles.UnpakBonDataPosition.BoneDataPosition)) {isSetParameters=true}
   &ScaleColor(
    Scale Alpha=&MakeCustomFloatFromBool(Bool=&SetBoolByFloatComparison(A=$Particles.Age, B=0.033), True Float=1, False Float=0)
    ScaleA=true
    ScaleRGB=true
   ) {module=ScaleColor004}
  ]
  rendererClasses [&NiagaraSpriteRendererProperties]
  renderers [{rendererClass=&NiagaraSpriteRendererProperties}]
 }
 Smoke_Wide_Large {
  emitterUpdateScript [
   &EmitterState(
    Enable Distance Culling=true
    Life Cycle Mode=::NewEnumerator1
    Loop Duration=1
    Module.bUseMaxDistance=true
    Module.bUseMinDistance=false
    Scalability Mode=::NewEnumerator1
    UseLoopDelay=false
   )
   &SpawnRate(
    SpawnRate=&Lerp_Float(
     A=0
     B=40
     Alpha=&FloatFromCurve(FloatCurve=DataInterface() {propertyValues {Curve {keys [<at 0 0>,<at 0.3 1>]}}}, CurveIndex=$System.Dive)
    )
    Use Spawn Probability=true
   )
  ]
  particleSpawnScript [
   &InitializeParticle(
    Color=&UniformRangedLinearColor(Minimum=<rgba 0.099 0.1043 0.11 1>, Maximum=<rgba 0.091 0.1098 0.13 1>)
    Lifetime=&UniformRangedFloat(Minimum=0.2, Maximum=0.35)
    Module.Write Color=true
    Module.Write Lifetime=true
    Module.Write Mass=true
    Module.Write Material Random=true
    Module.Write Position=true
    Module.Write Scale=false
    Module.Write SpriteRotation=true
    Module.Write SpriteSize=true
    Position=/Niagara/DynamicInputs/Helpers/SimulationPosition()
    Sprite Rotation=&NormalizedAngleToDegrees(NormalizedAngle=&UniformRangedFloat(Minimum=-90, Maximum=90))
    Sprite Size=&Vector2DFromFloat(Value=&RandomRangeFloat(Minimum=50, Maximum=60))
   )
   &FN_SetSpawnGroup_v2(Number of Spawn Groups=2)
   /SetVariables(Particles.InitSpriteRotRate=&UniformRangedFloat(Minimum=-3, Maximum=3)) {isSetParameters=true}
   /SetVariables(Particles.RandomFloat=&RandomRangeFloat(Minimum=0, Maximum=0.3)) {isSetParameters=true}
   /SetVariables(Particles.Velocity=<0 0 5000>) {isSetParameters=true}
  ]
  particleUpdateScript [
   &ParticleState()
   /SetVariables(Particles.Position=&Multiply_Vector(
     A=&Lerp_Vector(
      A=<0 0 30>
      B=<0 0 180>
      Alpha=&FloatFromCurve(
       FloatCurve=DataInterface() {propertyValues {Curve {keys [{interpMode=RCIM_Cubic, leaveTangent=0.1113, tangentMode=RCTM_Break, time=0, value=0}, <at 1 1>]}}}
       Scale Curve=&FloatFromCurve(
        FloatCurve=DataInterface() {propertyValues {Curve {keys [{interpMode=RCIM_Cubic, leaveTangent=2.2856, tangentMode=RCTM_Break, time=0, value=0.2}, <at 1 1>]}}}
        CurveIndex=$System.Dive
       )
      )
     )
     B=&VectorFromFloat(Value=&FloatFromCurve(FloatCurve=DataInterface() {propertyValues {Curve {keys [
           {arriveTangent=0.7642, interpMode=RCIM_Cubic, leaveTangent=0.7642, tangentMode=RCTM_User, time=0, value=0.2}
           {interpMode=RCIM_Cubic, tangentMode=RCTM_User, time=1, value=1}
          ]}}}))
    )) {isSetParameters=true}
   &SkeletalMeshLocation(
    Apply To Particle Position=false
    Bone / Socket Index=&SpawnGroup()
    Bone Sampling Mode=::NewEnumerator4
    Inherit Velocity=false
    Interpolate Mesh Sampling=false
    MaskSpawnGroup=false
    Skeletal Mesh=$User.SkeletalMesh
   ) {module=SkeletalMeshLocation001}
   &SpriteRotationRate(Rotation Rate=&Multiply_Float(
     A=$Particles.InitSpriteRotRate
     B=&FloatFromCurve(
      FloatCurve=DataInterface() {propertyValues {Curve {keys [
          {arriveTangent=-1.9214, interpMode=RCIM_Cubic, leaveTangent=-1.9214, tangentMode=RCTM_User, time=-0.0046, value=1.0113}
          {arriveTangent=-0.2496, interpMode=RCIM_Cubic, leaveTangent=-0.2496, tangentMode=RCTM_User, time=0.9896, value=0.16}
         ]}}}
      CurveIndex=$Emitter.Age
      Scale Curve=150
     )
    ))
   &ScaleSpriteSize(Scale Factor=&Vector2DFromFloat(Value=&FloatFromCurve(FloatCurve=DataInterface() {propertyValues {Curve {keys [
          {arriveTangent=0.7974, interpMode=RCIM_Cubic, leaveTangent=0.7974, tangentMode=RCTM_User, time=0, value=0.4656}
          {interpMode=RCIM_Cubic, leaveTangent=1.6352, time=0.7, value=1}
         ]}}})))
   &DynamicMaterialParameters(
    Index 0 Param 1=&Add_Float(
     A=$Particles.RandomFloat
     B=&FloatFromCurve(FloatCurve=DataInterface() {propertyValues {Curve {keys [
          {arriveTangent=0.9088, interpMode=RCIM_Cubic, leaveTangent=0.9088, tangentMode=RCTM_User, time=0, value=1}
          {arriveTangent=0.5473, interpMode=RCIM_Cubic, leaveTangent=0.5473, tangentMode=RCTM_User, time=1, value=2}
         ]}}})
    )
    Index 0 Param 2=&FloatFromCurve(FloatCurve=DataInterface() {propertyValues {Curve {keys [
         {arriveTangent=0.0453, interpMode=RCIM_Cubic, tangentMode=RCTM_Break, time=0, value=-1}
         {arriveTangent=0.0604, interpMode=RCIM_Cubic, leaveTangent=0.0604, tangentMode=RCTM_User, time=1, value=-1}
        ]}}})
    Index 0 Param 3=&FloatFromCurve(FloatCurve=DataInterface() {propertyValues {Curve {keys [
         {arriveTangent=-5.2387, interpMode=RCIM_Cubic, leaveTangent=-2.2846, tangentMode=RCTM_User, time=0, value=0.25}
         {interpMode=RCIM_Cubic, leaveTangent=4.6779, time=0.2851, value=0}
        ]}}})
    Index 1 Param 1=&FloatFromCurve(
     FloatCurve=DataInterface() {propertyValues {Curve {keys [<at 0 0>,<at 0.25 1>]}}}
     Scale Curve=&RandomRangeFloat(Minimum=2, Maximum=1)
    )
    Index 1 Param 2=&FloatFromCurve(
     FloatCurve=DataInterface() {propertyValues {Curve {keys [
         {interpMode=RCIM_Cubic, leaveTangent=3.7422, tangentMode=RCTM_Break, time=0, value=0.2}
         {interpMode=RCIM_Cubic, leaveTangent=0.9066, time=1, value=2}
        ]}}}
     Scale Curve=&RandomRangeFloat(Minimum=0.5, Maximum=1)
    )
    Index 1 Param 3=&FloatFromCurve(
     FloatCurve=DataInterface() {propertyValues {Curve {keys [
         {interpMode=RCIM_Cubic, leaveTangent=2.8445, tangentMode=RCTM_Break, time=0, value=0.2}
         {interpMode=RCIM_Cubic, leaveTangent=63.2978, time=0.25, value=1}
        ]}}}
     Scale Curve=0.8
    )
    Index 1 Param 4=&FloatFromCurve(
     FloatCurve=DataInterface() {propertyValues {Curve {keys [<at 0 1>,<at 1 0.8>]}}}
     Scale Curve=&RandomRangeFloat(Minimum=0.5, Maximum=1)
    )
    Index 2 Param 1=&FloatFromCurve(FloatCurve=DataInterface() {propertyValues {Curve {keys [{interpMode=RCIM_Cubic, leaveTangent=3.2309, tangentMode=RCTM_Break, time=0.1, value=0}, <at 1 1>]}}})
    Index 2 Param 2=&FloatFromCurve(
     FloatCurve=DataInterface() {propertyValues {Curve {keys [<at 0 0.2 RCIM_Cubic>,<at 1 0.8 RCIM_Cubic>]}}}
     Scale Curve=&RandomRangeFloat(Minimum=0.2, Maximum=1.2)
    )
    Index 2 Param 3=&FloatFromCurve(FloatCurve=DataInterface() {propertyValues {Curve {keys [
         {interpMode=RCIM_Cubic, leaveTangent=2.0952, tangentMode=RCTM_Break, time=0, value=1}
         {interpMode=RCIM_Cubic, leaveTangent=11.1233, tangentMode=RCTM_Break, time=1, value=4}
        ]}}})
    Index 2 Param 4=&FloatFromCurve(FloatCurve=DataInterface() {propertyValues {Curve {keys [<at 0 1 RCIM_Cubic>, {interpMode=RCIM_Cubic, leaveTangent=-1.2956, tangentMode=RCTM_Break, time=1, value=0.75}]}}})
    Module.Param0WriteEnabled=true
    Module.Param0WriteEnabled1=true
    Module.Param0WriteEnabled2=true
    Module.Param1WriteEnabled=false
    Module.Param1WriteEnabled1=true
    Module.Param1WriteEnabled2=true
    Module.Param2WriteEnabled=true
    Module.Param2WriteEnabled1=true
    Module.Param2WriteEnabled2=true
    Module.Param3WriteEnabled=true
    Module.Param3WriteEnabled1=true
    Module.Param3WriteEnabled2=true
    Write Parameter Index 1=true
    Write Parameter Index 2=true
   )
   /SetVariables(Particles.Orientation:Quat4f=&MultiplyQuaternion(
     Quaternion A=&FindQuatBetweenTwoVectors(From Vector=<0 0 1>, To Vector=&NormalizeVector(Vector To Normalize=$System.DeltaSystemPosition))
     Quaternion B=&FindQuatBetweenTwoVectors(
      From Vector=<0 0 1>
      To Vector=&NormalizeVector(Vector To Normalize=&MakeVector(
        X=&Multiply_Float(
         A=&SelectFloatFromArray(Direct Array Index=&SpawnGroup(), Float Selection Array=DataInterface() {propertyValues {FloatData [-0.5,0.5]}})
         B=&FloatFromCurve(
          FloatCurve=DataInterface() {propertyValues {Curve {keys [
              {interpMode=RCIM_Cubic, tangentMode=RCTM_Break, time=0, value=1}
              {arriveTangent=-2.2148, interpMode=RCIM_Cubic, tangentMode=RCTM_Break, time=1, value=0.2}
             ]}}}
          CurveIndex=$System.Dive
         )
        )
        Y=0
        Z=1
       ))
     )
    )) {isSetParameters=true}
   /SetVariables(Particles.Position=&Add_Vector(
     A=/Niagara/DynamicInputs/Math/TransformPositionByMatrix(
      Position:Vector3f=Linked() {linkedVariable {name=Particles.Position, type=NiagaraPosition}}
      TransformationMatrix:NiagaraMatrix=/Niagara/DynamicInputs/Math/MakeTransform(
       Inverse:NiagaraBool=false
       Position:Vector3f=?
       Local Offset:Vector3f=<0 0 0>
       Scale:Vector3f=<1 1 1>
       Rotation Interface:ENiagara_RotationDefinitionApproach=::NewEnumerator1
       Axis:Vector3f=?
       Quaternion:Quat4f=$Particles.Orientation
       Rotation Matrix:NiagaraMatrix=?
       X Vector:Vector3f=?
       Yaw:NiagaraFloat=?
       Angle:NiagaraFloat=?
       Pitch:NiagaraFloat=?
       Y Vector:Vector3f=?
       Roll:NiagaraFloat=?
       Z Vector:Vector3f=?
      )
     )
     B=$Output.SkeletalMeshLocation001.SampledPosition
    )) {isSetParameters=true}
   &ScaleColor(
    Scale Alpha=&FloatFromCurve(
     FloatCurve=DataInterface() {propertyValues {Curve {keys [
         {interpMode=RCIM_Cubic, tangentMode=RCTM_Break, time=0, value=0}
         {arriveTangent=4.4817, interpMode=RCIM_Cubic, leaveTangent=1.1637, tangentMode=RCTM_Break, time=0.4167, value=0.85}
         {interpMode=RCIM_Cubic, tangentMode=RCTM_Break, time=0.7498, value=0.9236}
         <at 1 0>
        ]}}}
     Scale Curve=&MakeCustomFloatFromBool(Bool=$User.bIsFrontend, True Float=1.25, False Float=1)
    )
    Scale RGB=&VectorFromFloat(Value=&FloatFromCurve(FloatCurve=DataInterface() {propertyValues {Curve {keys [
          {arriveTangent=0.2356, interpMode=RCIM_Cubic, leaveTangent=0.2356, tangentMode=RCTM_User, time=-0.2673, value=0.2914}
          {arriveTangent=1.471, interpMode=RCIM_Cubic, leaveTangent=1.471, tangentMode=RCTM_User, time=0.1335, value=0.5284}
          {arriveTangent=-4.0277, interpMode=RCIM_Cubic, tangentMode=RCTM_Break, time=0.4, value=0.8}
         ]}}}))
    ScaleA=true
    ScaleRGB=true
   )
   &ScaleColor(Scale Alpha=1.7, Scale RGB=<1.13 1.13 1.13>, ScaleA=true, ScaleRGB=true) {module=ScaleColor001}
   &ScaleColor(
    Scale Alpha=&FloatFromCurve(
     FloatCurve=DataInterface() {propertyValues {Curve {keys [
         {arriveTangent=0.2572, interpMode=RCIM_Cubic, leaveTangent=0.2572, tangentMode=RCTM_User, time=0, value=0}
         {arriveTangent=0.9315, interpMode=RCIM_Cubic, leaveTangent=0.9315, tangentMode=RCTM_User, time=0.5584, value=0.894}
         {interpMode=RCIM_Cubic, tangentMode=RCTM_User, time=1, value=1}
        ]}}}
     CurveIndex=$System.Dive
    )
    ScaleA=true
    ScaleRGB=true
   ) {module=ScaleColor002}
   &ScaleColor(
    Scale RGB=&MakeCustomVectorFromBool(
     Bool=$User.bIsFrontEndPreview
     True Vector=&MakeVectorFromLinearColorRGB(Color=&ColorFromCurve(ColorCurve=DataInterface() {propertyValues {
         AlphaCurve {keys [<at 0 1 RCIM_Cubic>,<at 1 0 RCIM_Cubic>]}
         BlueCurve {keys [<at 0 0.1759 RCIM_Cubic>,<at 1 4 RCIM_Cubic>]}
         GreenCurve {keys [<at 0 1 RCIM_Cubic>,<at 1 4 RCIM_Cubic>]}
         RedCurve {keys [<at 0 0 RCIM_Cubic>,<at 1 4 RCIM_Cubic>]}
        }}))
     False Vector=<1 1 1>
    )
    ScaleA=true
    ScaleRGB=true
   ) {module=ScaleColor003}
   &ScaleColor(
    Scale Alpha=&MakeCustomFloatFromBool(Bool=&SetBoolByFloatComparison(A=$Particles.Age, B=0.033), True Float=1, False Float=0)
    ScaleA=true
    ScaleRGB=true
   ) {module=ScaleColor004}
  ]
  rendererClasses [&NiagaraSpriteRendererProperties]
  renderers [{rendererClass=&NiagaraSpriteRendererProperties}]
 }
 Sparks {
  emitterUpdateScript [
   &EmitterState(
    Enable Distance Culling=true
    Life Cycle Mode=::NewEnumerator1
    Loop Duration=1
    MaxDistance=2000
    Module.bUseMaxDistance=true
    Module.bUseMinDistance=false
    Scalability Mode=::NewEnumerator1
    UseLoopDelay=false
   )
   &SpawnRate(
    SpawnRate=&Lerp_Float(
     A=0
     B=50
     Alpha=&FloatFromCurve(
      FloatCurve=DataInterface() {propertyValues {Curve {keys [{interpMode=RCIM_Cubic, leaveTangent=0.0843, tangentMode=RCTM_Break, time=0, value=0}, <at 0.7 1>]}}}
      CurveIndex=$System.Dive
     )
    )
    Use Spawn Probability=false
   )
  ]
  particleSpawnScript [
   &InitializeParticle(
    Color=<rgba 10 5.25 2.5 1>
    Lifetime=&RandomRangeFloat(Minimum=0.2, Maximum=0.4)
    Module.Write Color=true
    Module.Write Lifetime=true
    Module.Write Mass=true
    Module.Write Material Random=true
    Module.Write Position=false
    Module.Write Scale=false
    Module.Write SpriteRotation=false
    Module.Write SpriteSize=true
    Position=/Niagara/Functions/Localspace/SimulationPosition()
    Sprite Size=&Lerp_Vector2(
     A=<0.4 3.5434>
     B=<0.8 5.3135>
     Alpha=&FloatFromCurve(
      FloatCurve=DataInterface() {propertyValues {Curve {keys [<at 0 0>,<at 0.6657 0.1365>,<at 0.788 0.7387>,<at 1 1>]}}}
      CurveIndex=&RandomRangeFloat(Minimum=0, Maximum=1)
     )
    )
   )
   &AddVelocityInCone(
    Cone Angle=20
    Cone Axis=&SafeNormalizeVector(Vector To Normalize:Vector3f=$System.DeltaSystemPosition, Fallback Vector:Vector3f=<0 0 1>)
    Module.Use Velocity Falloff On Cone Axis=false
    Velocity Strength=&Multiply_Float(
     A=&RandomRangeFloat(Minimum=550, Maximum=950)
     B=&FloatFromCurve(
      FloatCurve=DataInterface() {propertyValues {Curve {keys [
          {arriveTangent=0.4749, interpMode=RCIM_Cubic, leaveTangent=0.4749, tangentMode=RCTM_User, time=0, value=0.3429}
          {interpMode=RCIM_Cubic, tangentMode=RCTM_User, time=0.7, value=1}
         ]}}}
      CurveIndex=$System.Dive
     )
    )
   )
   &SphereLocation(
    MaskSpawnGroup=false
    Override Seed=false
    Sphere Radius=10
    Surface Only=false
    UseNonUniformScale=false
    UseOffset=false
    UseSphereOrientation=false
   )
  ]
  particleUpdateScript [
   &SolveForcesAndVelocity(Clamp Velocity=false, Limit Acceleration=false)
   &ParticleState()
   &ScaleSpriteSize(Scale Factor=&Vector2DFromCurve(
     Vector2Curve=DataInterface() {propertyValues {
       XCurve {keys [<at 0 1>,<at 1 0>]}
       YCurve {keys [{arriveTangent=8.7888, interpMode=RCIM_Cubic, leaveTangent=8.7888, tangentMode=RCTM_User, time=0, value=1}, <at 1 7.0799>]}
      }}
     CurveIndex=&Multiply_Float(A=$Particles.NormalizedAge, B=2)
     ScaleCurve=<1 1>
    ))
   &FN_SkeletalBoneData(Index:NiagaraInt32=&MakeCustomIntFromBool(
     Bool=&SetBoolByIntComparison(
      A=&ModuloInt(A=&self:UniqueIDInt(), B=2)
      B=1
      Comparison Type=::NewEnumerator4
     )
     True INT=1
     False INT=0
    )) {module=FN_SkeletalBoneData001}
   &self:UnpakBonDataPosition(Input 1:ENiagaraBoneData=$Particles.FN_SkeletalBoneData001.BoneData) {module=UnpakBonDataPosition}
   /SetVariables(Particles.FinalPosition:Vector3f=&Add_Vector(A=Linked() {linkedVariable {name=Particles.Position, type=NiagaraPosition}}, B=$Particles.UnpakBonDataPosition.BoneDataPosition)) {isSetParameters=true}
   &ScaleColor(
    Scale Alpha=&FloatFromCurve(
     FloatCurve=DataInterface() {propertyValues {Curve {keys [<at 0 1>,<at 1 0>]}}}
     CurveIndex=&Multiply_Float(A=$Particles.NormalizedAge, B=2)
    )
    Scale RGB=&MakeVectorFromLinearColorRGB(Color=&ColorFromCurve(
      ColorCurve=DataInterface() {propertyValues {
        AlphaCurve {keys []}
        BlueCurve {keys [<at 0 0.203 RCIM_Cubic>,<at 0.5 0>,<at 1 0.875 RCIM_Cubic>]}
        GreenCurve {keys [<at 0 1 RCIM_Cubic>,<at 0.5 1>,<at 1 0.9208 RCIM_Cubic>]}
        RedCurve {keys [<at 0 0.4609 RCIM_Cubic>,<at 0.5 0.2356>,<at 1 1 RCIM_Cubic>]}
       }}
      CurveIndex=&Multiply_Float(A=$Particles.NormalizedAge, B=2)
     ))
    ScaleA=true
    ScaleRGB=true
   )
   &ScaleColor(
    Scale Alpha=&FloatFromCurve(
     FloatCurve=DataInterface() {propertyValues {Curve {keys [
         {arriveTangent=0.3778, interpMode=RCIM_Cubic, leaveTangent=0.3778, tangentMode=RCTM_User, time=0, value=0.3714}
         {arriveTangent=0.8381, interpMode=RCIM_Cubic, leaveTangent=0.8381, time=0.4386, value=0.8705}
         {interpMode=RCIM_Cubic, tangentMode=RCTM_User, time=0.75, value=1}
        ]}}}
     CurveIndex=$System.Dive
    )
    ScaleA=true
    ScaleRGB=true
   ) {module=ScaleColor001}
  ]
  rendererClasses [&NiagaraSpriteRendererProperties]
  renderers [{rendererClass=&NiagaraSpriteRendererProperties}]
 }
 SwingChain {
  simTarget=GPUComputeSim
  emitterSpawnScript [
   /SetVariables(Emitter.SegLength=5.4, Emitter.SegNum=30) {isSetParameters=true, module=SetVariables_#11}
   /SetVariables(Emitter.Randomer=0.3727, Emitter.Randomer2=0.8) {isSetParameters=true, module=SetVariables_#12}
   /SetVariables(Emitter.SwingStrengthCurve=DataInterface() {propertyValues {Curve {keys [
        {interpMode=RCIM_Cubic, tangentMode=RCTM_Break, time=0, value=0}
        {arriveTangent=1.204, interpMode=RCIM_Cubic, leaveTangent=1.204, tangentMode=RCTM_User, time=0.3331, value=0.4163}
        {arriveTangent=0.6146, interpMode=RCIM_Cubic, tangentMode=RCTM_Break, time=1, value=1}
       ]}}}) {isSetParameters=true}
   /SetVariables(Emitter.SwingStrengthCurve_FE=DataInterface() {propertyValues {Curve {keys [
        {interpMode=RCIM_Cubic, leaveTangent=0.0361, tangentMode=RCTM_Break, time=0, value=0}
        {arriveTangent=1.6799, interpMode=RCIM_Cubic, leaveTangent=1.6799, tangentMode=RCTM_User, time=0.2749, value=0.5702}
        {interpMode=RCIM_Cubic, tangentMode=RCTM_Break, time=1, value=1}
       ]}}}) {isSetParameters=true}
  ]
  emitterUpdateScript [
   &EmitterState(
    Enable Distance Culling=true
    Inactive Response=::NewEnumerator1
    Life Cycle Mode=::NewEnumerator1
    Loop Behavior=::NewEnumerator1
    Loop Duration Mode=::NewEnumerator1
    MaxDistance=7000
    Module.bUseMaxDistance=true
    Module.bUseMinDistance=false
    Scalability Mode=::NewEnumerator1
    UseLoopDelay=false
   )
   &SpawnBurst_Instantaneous(Spawn Count=$Emitter.SegNum, Use Spawn Probability=false)
   &SpawnBurst_Instantaneous(Spawn Count=$Emitter.SegNum, Spawn Group=1, Use Spawn Probability=false) {module=SpawnBurst_Instantaneous001}
   /SetVariables(Emitter.RemapDive:NiagaraFloat=&FloatFromCurve(
     FloatCurve=DataInterface() {propertyValues {Curve {keys [{interpMode=RCIM_Cubic, leaveTangent=0.7295, tangentMode=RCTM_Break, time=0, value=0}, {tangentMode=RCTM_User, time=0.8, value=1}]}}}
     CurveIndex=$System.Dive
    )) {isSetParameters=true, module=SetVariables_#13}
  ]
  particleSpawnScript [
   /Niagara/Modules/Spawn/Initialization/V2/InitializeParticle(Color Mode=::NewEnumerator0, UsePositionOffset=false, Write Material Random=true)
   /SetVariables(Particles.SpawnGroup=$Emitter.SpawnGroup) {isSetParameters=true, module=SetVariables_#14}
   /SetVariables(Particles.MUniqueID:NiagaraInt32=&ModuloInt(A=&self:UniqueIDInt(), B=$Emitter.SegNum)) {isSetParameters=true}
   &AddVelocity(Velocity=&Lerp_Vector(
     A=<0 0 0>
     B=&MakeCustomVectorFromBool(Bool=$User.bIsFrontEndPreview, True Vector=<119.0033 109.8097 614.8258>, False Vector=<1200 1200 1500>)
     Alpha=&FloatFromCurve(
      FloatCurve=DataInterface() {propertyValues {Curve {keys [
          {interpMode=RCIM_Cubic, leaveTangent=0.5081, tangentMode=RCTM_Break, time=0, value=0}
          {arriveTangent=2.0274, interpMode=RCIM_Cubic, tangentMode=RCTM_Break, time=1, value=1}
         ]}}}
      CurveIndex=&Divide_Float(
       A=&MakeFloatFromInt(Int=$Particles.MUniqueID)
       B=&MakeFloatFromInt(Int=$Emitter.SegNum)
      )
     )
    ))
  ]
  particleUpdateScript [
   &ParticleState(Kill Particles When Lifetime Has Elapsed=false, Let Infinitely Lived Particles Die When Emitter Deactivates=true)
   /SetVariables(Particles.Position=<0 0 0>) {isSetParameters=true, module=SetVariables_#15}
   &SkeletalMeshLocation(
    Apply To Particle Position=false
    Bone Sampling Mode=::NewEnumerator4
    Inherit Velocity=false
    Interpolate Mesh Sampling=false
    MaskSpawnGroup=false
    Sampled Position Offset=&Lerp_Vector(A=<5 0 -6>, B=<0 0 0>, Alpha=$System.Dive)
    Sampled Position Offset Coordinate Space:ENiagaraMeshTransforms=::NewEnumerator3
    Skeletal Mesh=$User.SkeletalMesh
   )
   &SkeletalMeshLocation(
    Apply To Particle Position=false
    Bone / Socket Index=1
    Bone Sampling Mode=::NewEnumerator4
    Inherit Velocity=false
    Interpolate Mesh Sampling=false
    MaskSpawnGroup=false
    Sampled Position Offset=&Lerp_Vector(A=<5 0 -6>, B=<0 0 0>, Alpha=$System.Dive)
    Sampled Position Offset Coordinate Space:ENiagaraMeshTransforms=::NewEnumerator3
    Skeletal Mesh=$User.SkeletalMesh
   ) {module=SkeletalMeshLocation001}
   &self:InitialChain(
    CurrentSegNum:NiagaraInt32=&MakeIntFromFloat(Float=&Lerp_Float(
      A=13
      B=&MakeFloatFromInt(Int=$Emitter.SegNum)
      Alpha=&Multiply_Float(A=$Emitter.RemapDive, B=$System.ADive.ADive)
     ))
    Distance:NiagaraFloat=$Emitter.SegLength
    UniqueID:NiagaraInt32=$Particles.MUniqueID
   ) {module=InitialChain}
   &self:CurlNoise(
    Speed=1.5
    Timer:NiagaraFloat=&Add_Float(A=$Emitter.Age, B=&MaskFloatBySpawnGroup(Masked Float=$Emitter.Randomer, Passthrough Float=$Emitter.Randomer2))
   ) {module=CurlNoise}
   &self:Timer(ScaleDeltaTime:NiagaraFloat=&Multiply_Float(
     A=&Lerp_Float(A=-5, B=-6, Alpha=$Emitter.RemapDive)
     B=&Lerp_Float(A=&Lerp_Float(A=0.5, B=0.75, Alpha=$Emitter.RemapDive), B=1.3, Alpha=$Particles.CurlNoise.X)
    )) {module=Timer}
   &self:Timer(ScaleDeltaTime:NiagaraFloat=&Multiply_Float(
     A=&Lerp_Float(A=-3.5, B=-5.5, Alpha=$Emitter.RemapDive)
     B=&Lerp_Float(A=&Lerp_Float(A=0.5, B=0.75, Alpha=$Emitter.RemapDive), B=1.3, Alpha=$Particles.CurlNoise.Y)
    )) {module=Timer001}
   &self:ChainColor_PerParticle(
    ColorCurve:NiagaraDataInterfaceColorCurve=$System.SwingChainColor
    MaxSeg:NiagaraInt32=&MakeIntFromFloat(Float=&FloatFromCurve(FloatCurve=DataInterface() {propertyValues {Curve {keys [<at 0 0.4>,<at 1 1>]}}}, CurveIndex=$Emitter.RemapDive, Scale Curve=55))
    UniqueID:NiagaraInt32=$Particles.MUniqueID
   ) {module=ChainColor_PerParticle}
   /SetVariables(Particles.Color=&MultiplyLinearColor(
     Linear Color=$Particles.Color
     Scale Factor=&MakeVector4_FromFloat(Float=&FloatFromCurve(
       FloatCurve=DataInterface() {propertyValues {Curve {keys [
           {interpMode=RCIM_Cubic, leaveTangent=0.1727, tangentMode=RCTM_Break, time=0, value=0.1}
           {arriveTangent=0.1312, interpMode=RCIM_Cubic, leaveTangent=0.1312, tangentMode=RCTM_User, time=0.5009, value=0.18}
           {arriveTangent=0.0672, interpMode=RCIM_Cubic, tangentMode=RCTM_Break, time=1, value=0.25}
          ]}}}
       CurveIndex=$Emitter.RemapDive
       Scale Curve=2
      ))
    )) {isSetParameters=true}
   &self:SineWave(
    Axis:Vector3f=&MaskVectorBySpawnGroup(Masked Vector=<1 -1 0>, Passthrough Vector=<1 1 0>)
    Period:NiagaraFloat=&Lerp_Float(A=0.06, B=0.04, Alpha=$Emitter.RemapDive)
    Randomer:NiagaraFloat=$Emitter.Randomer
    SegLength:NiagaraFloat=$Emitter.SegLength
    SegNumOfFalloff=15
    StrengthFalloff:NiagaraDataInterfaceCurve=$Emitter.SwingStrengthCurve
    SwingStrengthX=12
    SwingStrengthY=8
    TimerX:NiagaraFloat=$Particles.Timer.Timer
    TimerY:NiagaraFloat=$Particles.Timer001.Timer
    UniqueID:NiagaraInt32=$Particles.MUniqueID
   ) {module=SineWave}
   /SetVariables(Particles.MeshOrientation=&MultiplyQuaternion(
     Quaternion A=&MakeQuatFromAxisAngle(
      Axis=$Particles.MeshZAxisAfterWaveOrient
      AngleInDegrees=&Multiply_Float(A=25, B=&MakeFloatFromInt(Int=$Particles.MUniqueID))
     )
     Quaternion B=$Particles.WaveOrient
    )) {isSetParameters=true}
   &self:ChainBend(
    BendCurve=DataInterface() {propertyValues {Curve {keys [
        {interpMode=RCIM_Cubic, leaveTangent=-2.9396, tangentMode=RCTM_Break, time=0, value=0}
        {arriveTangent=-0.5, interpMode=RCIM_Cubic, leaveTangent=-0.5, time=0.2, value=-0.3}
        {interpMode=RCIM_Cubic, tangentMode=RCTM_Break, time=1, value=-0.5}
       ]}}}
    BendStrength:NiagaraFloat=&MaskFloatBySpawnGroup(Masked Float=70, Passthrough Float=-70)
    BendStrength2:NiagaraFloat=&FloatFromCurve(
     FloatCurve=DataInterface() {propertyValues {Curve {keys [
         {interpMode=RCIM_Cubic, tangentMode=RCTM_Break, time=0, value=1}
         {arriveTangent=-2.7852, interpMode=RCIM_Cubic, tangentMode=RCTM_Break, time=1, value=0.2}
        ]}}}
     CurveIndex=$Emitter.RemapDive
    )
    Core:Vector3f=$Particles.WavePosition
    MaxSeg=40
    SegLength:NiagaraFloat=$Emitter.SegLength
    UniqueID:NiagaraInt32=$Particles.MUniqueID
   ) {module=ChainBend}
   /SetVariables(Particles.MeshOrientation=&MultiplyQuaternion(Quaternion A=$Particles.BendOrientation, Quaternion B=$Particles.MeshOrientation)) {isSetParameters=true}
   &self:TotalRotation(
    NewLookAt:Vector3f=&NormalizeVector(Vector To Normalize=$System.DeltaSystemPosition)
    Position:NiagaraPosition=$Particles.Position
   ) {module=TotalRotation}
   /SetVariables(Particles.Position=&Add_Vector(
     A=Linked() {linkedVariable {name=Particles.Position, type=NiagaraPosition}}
     B=&MaskVectorBySpawnGroup(Masked Vector=$Output.SkeletalMeshLocation.SampledPosition, Passthrough Vector=$Output.SkeletalMeshLocation001.SampledPosition)
    )) {isSetParameters=true, module=SetVariables_#16}
  ]
  rendererClasses [&NiagaraMeshRendererProperties]
  renderers [{rendererClass=&NiagaraMeshRendererProperties, rendererIndex=0}]
 }
 SwingChain001 {
  emitterSpawnScript [
   /SetVariables(Emitter.SegLength=5.4, Emitter.SegNum=30) {isSetParameters=true, module=SetVariables_#11}
   /SetVariables(Emitter.Randomer=0.3727, Emitter.Randomer2=0.8) {isSetParameters=true, module=SetVariables_#12}
  ]
  emitterUpdateScript [
   &EmitterState(
    Enable Distance Culling=true
    Inactive Response=::NewEnumerator1
    Life Cycle Mode=::NewEnumerator1
    Loop Behavior=::NewEnumerator1
    Loop Duration Mode=::NewEnumerator1
    MaxDistance=7000
    Module.bUseMaxDistance=true
    Module.bUseMinDistance=false
    Scalability Mode=::NewEnumerator1
    UseLoopDelay=false
   )
   &SpawnBurst_Instantaneous(Spawn Count=1, Use Spawn Probability=false)
   &SpawnBurst_Instantaneous(Spawn Count=1, Spawn Group=1, Use Spawn Probability=false) {module=SpawnBurst_Instantaneous001}
   /SetVariables(Emitter.RemapDive:NiagaraFloat=&FloatFromCurve(
     FloatCurve=DataInterface() {propertyValues {Curve {keys [{interpMode=RCIM_Cubic, leaveTangent=0.7295, tangentMode=RCTM_Break, time=0, value=0}, {tangentMode=RCTM_User, time=0.8, value=1}]}}}
     CurveIndex=$System.Dive
    )) {isSetParameters=true, module=SetVariables_#13}
  ]
  particleSpawnScript [
   &InitializeParticle(Color Mode=::NewEnumerator0, UsePositionOffset=false, Write Material Random=true)
   /SetVariables(Particles.SpawnGroup=$Emitter.SpawnGroup) {isSetParameters=true, module=SetVariables_#14}
  ]
  particleUpdateScript [
   &ParticleState(Kill Particles When Lifetime Has Elapsed=false, Let Infinitely Lived Particles Die When Emitter Deactivates=true)
   /SetVariables(Particles.Position=<0 0 0>) {isSetParameters=true, module=SetVariables_#15}
   &SkeletalMeshLocation(
    Apply To Particle Position=false
    Bone Sampling Mode=::NewEnumerator4
    Inherit Velocity=false
    Interpolate Mesh Sampling=false
    MaskSpawnGroup=false
    Sampled Position Offset=&Lerp_Vector(A=<5 0 -6>, B=<0 0 0>, Alpha=$System.Dive)
    Sampled Position Offset Coordinate Space:ENiagaraMeshTransforms=::NewEnumerator3
    Skeletal Mesh=$User.SkeletalMesh
   )
   &SkeletalMeshLocation(
    Apply To Particle Position=false
    Bone / Socket Index=1
    Bone Sampling Mode=::NewEnumerator4
    Inherit Velocity=false
    Interpolate Mesh Sampling=false
    MaskSpawnGroup=false
    Sampled Position Offset=&Lerp_Vector(A=<5 0 -6>, B=<0 0 0>, Alpha=$System.Dive)
    Sampled Position Offset Coordinate Space:ENiagaraMeshTransforms=::NewEnumerator3
    Skeletal Mesh=$User.SkeletalMesh
   ) {module=SkeletalMeshLocation001}
   &self:CurlNoise(
    Speed=1.5
    Timer:NiagaraFloat=&Add_Float(A=$Emitter.Age, B=&MaskFloatBySpawnGroup(Masked Float=$Emitter.Randomer, Passthrough Float=$Emitter.Randomer2))
   ) {module=CurlNoise}
   &self:Timer(ScaleDeltaTime:NiagaraFloat=&Multiply_Float(
     A=&Lerp_Float(A=-5, B=-6, Alpha=$Emitter.RemapDive)
     B=&Lerp_Float(A=&Lerp_Float(A=0.5, B=0.75, Alpha=$Emitter.RemapDive), B=1.3, Alpha=$Particles.CurlNoise.X)
    )) {module=Timer}
   &self:Timer(ScaleDeltaTime:NiagaraFloat=&Multiply_Float(
     A=&Lerp_Float(A=-3.5, B=-5.5, Alpha=$Emitter.RemapDive)
     B=&Lerp_Float(A=&Lerp_Float(A=0.5, B=0.75, Alpha=$Emitter.RemapDive), B=1.3, Alpha=$Particles.CurlNoise.Y)
    )) {module=Timer001}
   /SetVariables(Particles.Position=&Add_Vector(
     A=Linked() {linkedVariable {name=Particles.Position, type=NiagaraPosition}}
     B=&MaskVectorBySpawnGroup(Masked Vector=$Output.SkeletalMeshLocation.SampledPosition, Passthrough Vector=$Output.SkeletalMeshLocation001.SampledPosition)
    )) {isSetParameters=true, module=SetVariables_#16}
   /SetVariables(Particles.MeshOrientation=&FindQuatBetweenTwoVectors(From Vector=<0 0 1>, To Vector=&NormalizeVector(Vector To Normalize=$System.DeltaSystemPosition))) {isSetParameters=true}
   /SetVariables(Particles.MeshOrientation=&MultiplyQuaternion(
     Quaternion A=&FindQuatBetweenTwoVectors(
      From Vector=<0 0 -1>
      To Vector=&NormalizeVector(Vector To Normalize=&MakeVector(
        X=&Lerp_Float(A=&MaskFloatBySpawnGroup(Masked Float=0.3, Passthrough Float=-0.3), B=0, Alpha=$Emitter.RemapDive)
        Y=0
        Z=-1
       ))
     )
     Quaternion B=$Particles.MeshOrientation
    )) {isSetParameters=true}
   &DynamicMaterialParameters(
    Index 0 Param 1=1
    Index 0 Param 4=&Divide_Float(
     A=&Lerp_Float(
      A=&Subtract_Float(A=&MakeFloatFromInt(Int=$Emitter.SegNum), B=14)
      B=0
      Alpha=&Multiply_Float(A=$Emitter.RemapDive, B=$System.ADive.ADive)
     )
     B=&Multiply_Float(A=&MakeFloatFromInt(Int=$Emitter.SegNum), B=-1)
    )
    Index 1 Param 2=1
    Index 2 Param 1=$Particles.Timer.Timer
    Index 2 Param 2=&Lerp_Float(A=-3, B=-1.3, Alpha=$Emitter.RemapDive)
    Index 2 Param 3=12
    Index 3 Param 1=$Particles.Timer001.Timer
    Index 3 Param 2=&Lerp_Float(A=-3, B=-1.3, Alpha=$Emitter.RemapDive)
    Index 3 Param 3=8
    Module.Param0WriteEnabled=true
    Module.Param0WriteEnabled1=true
    Module.Param0WriteEnabled2=true
    Module.Param0WriteEnabled3=true
    Module.Param1WriteEnabled=true
    Module.Param1WriteEnabled1=true
    Module.Param1WriteEnabled2=true
    Module.Param1WriteEnabled3=true
    Module.Param2WriteEnabled=true
    Module.Param2WriteEnabled1=true
    Module.Param2WriteEnabled2=true
    Module.Param2WriteEnabled3=true
    Module.Param3WriteEnabled=true
    Module.Param3WriteEnabled1=true
    Module.Param3WriteEnabled2=true
    Module.Param3WriteEnabled3=true
    Write Parameter Index 1=true
    Write Parameter Index 2=true
    Write Parameter Index 3=true
   )
   &self:CalcXYAxisWS() {module=CalcXYAxisWS}
   &ScaleColor(
    Scale RGB=&VectorFromFloat(Value=&Lerp_Float(A=0.1, B=1, Alpha=$Emitter.RemapDive))
    ScaleA=true
    ScaleRGB=true
   )
  ]
  rendererClasses [&NiagaraMeshRendererProperties]
  renderers [{rendererClass=&NiagaraMeshRendererProperties, rendererIndex=0}]
 }
 WarpChain {
  simTarget=GPUComputeSim
  emitterSpawnScript [/SetVariables(Emitter.SegNum=75) {isSetParameters=true}]
  emitterUpdateScript [
   &EmitterState(
    Enable Distance Culling=true
    Inactive Response=::NewEnumerator1
    Life Cycle Mode=::NewEnumerator1
    Loop Behavior=::NewEnumerator1
    Loop Duration Mode=::NewEnumerator1
    MaxDistance=7000
    Module.bUseMaxDistance=true
    Module.bUseMinDistance=false
    Scalability Mode=::NewEnumerator1
    UseLoopDelay=false
   )
   /SetVariables(Emitter.RemapDive:NiagaraFloat=&FloatFromCurve(
     FloatCurve=DataInterface() {propertyValues {Curve {keys [
         {arriveTangent=0.2767, interpMode=RCIM_Cubic, leaveTangent=0.2767, tangentMode=RCTM_User, time=0, value=0}
         {arriveTangent=1.8467, interpMode=RCIM_Cubic, leaveTangent=1.8467, tangentMode=RCTM_User, time=1, value=1}
        ]}}}
     CurveIndex=$System.Dive
    )) {isSetParameters=true}
   /SetVariables(Emitter.ShowSegCount:NiagaraInt32=&MakeIntFromFloat(Float=&Lerp_Float(A=42, B=12, Alpha=&Multiply_Float(A=$Emitter.RemapDive, B=$System.ADive.ADive)))) {isSetParameters=true, module=SetVariables_#17}
   &SpawnBurst_Instantaneous(Spawn Count=$Emitter.SegNum, Use Spawn Probability=false)
   &SpawnBurst_Instantaneous(Spawn Count=$Emitter.SegNum, Spawn Group=1, Use Spawn Probability=false) {module=SpawnBurst_Instantaneous001}
  ]
  particleSpawnScript [
   &InitializeParticle(UsePositionOffset=false, Write Material Random=true)
   /SetVariables(Particles.SpawnGroup=$Emitter.SpawnGroup) {isSetParameters=true, module=SetVariables_#18}
   /SetVariables(Particles.MUniqueID:NiagaraInt32=&ModuloInt(A=&self:UniqueIDInt(), B=$Emitter.SegNum)) {isSetParameters=true}
  ]
  particleUpdateScript [
   &ParticleState(Kill Particles When Lifetime Has Elapsed=false, Let Infinitely Lived Particles Die When Emitter Deactivates=true)
   &SkeletalMeshLocation(
    Apply Sampled Orientation to Mesh Particles=true
    Apply To Particle Position:NiagaraBool=&SetBoolByIntComparison(A=&SpawnGroup(), B=0, Comparison Type=::NewEnumerator4)
    Bone Sampling Mode=::NewEnumerator4
    Inherit Velocity=false
    Interpolate Mesh Sampling=false
    MaskSpawnGroup=false
    Skeletal Mesh=$User.SkeletalMesh
   )
   &SkeletalMeshLocation(
    Apply Sampled Orientation to Mesh Particles=true
    Apply To Particle Position:NiagaraBool=&SetBoolByIntComparison(A=&SpawnGroup(), B=1, Comparison Type=::NewEnumerator4)
    Bone / Socket Index=1
    Bone Sampling Mode=::NewEnumerator4
    Inherit Velocity=false
    Interpolate Mesh Sampling=false
    MaskSpawnGroup=false
    Skeletal Mesh=$User.SkeletalMesh
   ) {module=SkeletalMeshLocation002}
   &SkeletalMeshLocation(
    Apply Sampled Orientation to Mesh Particles=true
    Apply To Particle Position=false
    Bone / Socket Index=5
    Bone Sampling Mode=::NewEnumerator4
    Inherit Velocity=false
    Interpolate Mesh Sampling=false
    MaskSpawnGroup=false
    Skeletal Mesh=$User.SkeletalMesh
   ) {module=SkeletalMeshLocation001}
   &SkeletalMeshLocation(
    Apply Sampled Orientation to Mesh Particles=true
    Apply To Particle Position=false
    Bone / Socket Index=6
    Bone Sampling Mode=::NewEnumerator4
    Inherit Velocity=false
    Interpolate Mesh Sampling=false
    MaskSpawnGroup=false
    Skeletal Mesh=$User.SkeletalMesh
   ) {module=SkeletalMeshLocation003}
   &self:Circle(
    AngleFixRoll=90
    AngleFixYaw=20
    Bone1Pos:Vector3f=&MaskVectorBySpawnGroup(Masked Vector=$Output.SkeletalMeshLocation.SampledPosition, Passthrough Vector=$Output.SkeletalMeshLocation002.SampledPosition)
    Bone2Pos:Vector3f=&MaskVectorBySpawnGroup(Masked Vector=$Output.SkeletalMeshLocation001.SampledPosition, Passthrough Vector=$Output.SkeletalMeshLocation003.SampledPosition)
    CurveOfRadiusScale=DataInterface() {propertyValues {Curve {keys [<at 0 0.7>,<at 1 1>]}}}
    InvertCircleDirct:NiagaraFloat=&MaskFloatBySpawnGroup(Masked Float=-1, Passthrough Float=1)
    OverRotate:NiagaraFloat=&Lerp_Float(A=0, B=1, Alpha=$System.Dive)
    Period=0.88
    PitchGradient=0.163
    Radius=7
    RadiusRecalcSegNum=20
    SegNum:NiagaraInt32=$Emitter.ShowSegCount
    Spacing=0.0727
    UniqueID:NiagaraInt32=$Particles.MUniqueID
   ) {module=Circle}
   &self:ChainColor_PerParticle(
    ColorCurve:NiagaraDataInterfaceColorCurve=$System.WarpChainColor
    MaxSeg:NiagaraInt32=&MakeIntFromFloat(Float=136.8021)
    UniqueID:NiagaraInt32=$Particles.MUniqueID
   ) {module=ChainColor_PerParticle}
   /SetVariables(Particles.FinalColor:LinearColor=&MultiplyLinearColor(
     Linear Color=$Particles.Color
     Scale Factor=&MakeVector4_FromFloat(Float=&FloatFromCurve(
       FloatCurve=DataInterface() {propertyValues {Curve {keys [
           {interpMode=RCIM_Cubic, leaveTangent=0.1727, tangentMode=RCTM_Break, time=0, value=0.1}
           {arriveTangent=0.1312, interpMode=RCIM_Cubic, leaveTangent=0.1312, tangentMode=RCTM_User, time=0.5009, value=0.1805}
           {arriveTangent=0.0672, interpMode=RCIM_Cubic, tangentMode=RCTM_Break, time=1, value=0.25}
          ]}}}
       CurveIndex=$System.Dive
       Scale Curve=2
      ))
    )) {isSetParameters=true}
   /SetVariables(Particles.Scale=&Multiply_Vector(A=$Particles.Scale, B=<1.2 1.2 1>)) {isSetParameters=true}
  ]
  rendererClasses [&NiagaraMeshRendererProperties]
  renderers [{rendererClass=&NiagaraMeshRendererProperties, rendererIndex=0}]
 }
 WarpChain001 {
  emitterUpdateScript [
   &EmitterState(
    Enable Distance Culling=true
    Inactive Response=::NewEnumerator1
    Life Cycle Mode=::NewEnumerator1
    Loop Behavior=::NewEnumerator1
    Loop Duration Mode=::NewEnumerator1
    MaxDistance=7000
    Module.bUseMaxDistance=true
    Module.bUseMinDistance=false
    Scalability Mode=::NewEnumerator1
    UseLoopDelay=false
   )
   /SetVariables(Emitter.ShowSegCount:NiagaraInt32=&MakeIntFromFloat(Float=&Lerp_Float(A=42, B=12, Alpha=&Multiply_Float(A=$Emitter.RemapDive, B=$System.ADive.ADive)))) {isSetParameters=true, module=SetVariables_#17}
   &SpawnBurst_Instantaneous(Spawn Count=1, Use Spawn Probability=false)
   &SpawnBurst_Instantaneous(Spawn Count=1, Spawn Group=1, Use Spawn Probability=false) {module=SpawnBurst_Instantaneous001}
   /SetVariables(Emitter.RemapDive:NiagaraFloat=&FloatFromCurve(
     FloatCurve=DataInterface() {propertyValues {Curve {keys [
         {arriveTangent=0.2767, interpMode=RCIM_Cubic, leaveTangent=0.2767, tangentMode=RCTM_User, time=0, value=0}
         {arriveTangent=1.8467, interpMode=RCIM_Cubic, leaveTangent=1.8467, tangentMode=RCTM_User, time=1, value=1}
        ]}}}
     CurveIndex=$System.Dive
    )) {isSetParameters=true}
  ]
  particleSpawnScript [
   &InitializeParticle(UsePositionOffset=false, Write Material Random=true)
   /SetVariables(Particles.SpawnGroup=$Emitter.SpawnGroup) {isSetParameters=true, module=SetVariables_#18}
  ]
  particleUpdateScript [
   &ParticleState(Kill Particles When Lifetime Has Elapsed=false, Let Infinitely Lived Particles Die When Emitter Deactivates=true)
   &SkeletalMeshLocation(
    Apply Sampled Orientation to Mesh Particles=true
    Apply To Particle Position:NiagaraBool=&SetBoolByIntComparison(A=&SpawnGroup(), B=0, Comparison Type=::NewEnumerator4)
    Bone Sampling Mode=::NewEnumerator4
    Inherit Velocity=false
    Interpolate Mesh Sampling=false
    MaskSpawnGroup=false
    Skeletal Mesh=$User.SkeletalMesh
   )
   &SkeletalMeshLocation(
    Apply Sampled Orientation to Mesh Particles=true
    Apply To Particle Position:NiagaraBool=&SetBoolByIntComparison(A=&SpawnGroup(), B=1, Comparison Type=::NewEnumerator4)
    Bone / Socket Index=1
    Bone Sampling Mode=::NewEnumerator4
    Inherit Velocity=false
    Interpolate Mesh Sampling=false
    MaskSpawnGroup=false
    Skeletal Mesh=$User.SkeletalMesh
   ) {module=SkeletalMeshLocation002}
   &SkeletalMeshLocation(
    Apply Sampled Orientation to Mesh Particles=true
    Apply To Particle Position=false
    Bone / Socket Index=5
    Bone Sampling Mode=::NewEnumerator4
    Inherit Velocity=false
    Interpolate Mesh Sampling=false
    MaskSpawnGroup=false
    Skeletal Mesh=$User.SkeletalMesh
   ) {module=SkeletalMeshLocation001}
   &SkeletalMeshLocation(
    Apply Sampled Orientation to Mesh Particles=true
    Apply To Particle Position=false
    Bone / Socket Index=6
    Bone Sampling Mode=::NewEnumerator4
    Inherit Velocity=false
    Interpolate Mesh Sampling=false
    MaskSpawnGroup=false
    Skeletal Mesh=$User.SkeletalMesh
   ) {module=SkeletalMeshLocation003}
   /SetVariables(Particles.Position=<0 0 0>) {isSetParameters=true}
   /SetVariables(
    Particles.MeshOrientation=&FindQuatBetweenTwoVectors(
     From Vector=<0 0 1>
     To Vector=&MaskVectorBySpawnGroup(
      Masked Vector=&NormalizeVector(Vector To Normalize=&Subtract_Vector(A=$Output.SkeletalMeshLocation.SampledPosition, B=$Output.SkeletalMeshLocation001.SampledPosition))
      Passthrough Vector=&NormalizeVector(Vector To Normalize=&Subtract_Vector(A=$Output.SkeletalMeshLocation002.SampledPosition, B=$Output.SkeletalMeshLocation003.SampledPosition))
     )
    )
    Particles.Position=&MaskVectorBySpawnGroup(Masked Vector=$Output.SkeletalMeshLocation001.SampledPosition, Passthrough Vector=$Output.SkeletalMeshLocation003.SampledPosition)
   ) {isSetParameters=true}
   /SetVariables(Particles.MeshOrientation=&MultiplyQuaternion(
     Quaternion A=$Particles.MeshOrientation
     Quaternion B=&MakeQuatFromAxisAngle(
      Axis=<0 0 1>
      AngleInDegrees=&Multiply_Float(A=720, B=&Multiply_Float(A=$Emitter.RemapDive, B=$System.ADive.ADive))
     )
    )) {isSetParameters=true}
   &DynamicMaterialParameters(
    Index 0 Param 4=HlslExpression() {hlslExpression="(1.0-float(Emitter.ShowSegCount)/42.0)*-1.0 /* Custom HLSL! */"}
    Module.Param0WriteEnabled=true
    Module.Param1WriteEnabled=true
    Module.Param2WriteEnabled=true
    Module.Param3WriteEnabled=true
   )
   &ScaleColor(
    Scale RGB=&VectorFromFloat(Value=&Lerp_Float(A=0.15, B=1, Alpha=$Emitter.RemapDive))
    ScaleA=true
    ScaleRGB=true
   )
   /SetVariables(Particles.Scale=<1.1 1.1 1>) {isSetParameters=true}
  ]
  rendererClasses [&NiagaraMeshRendererProperties]
  renderers [{rendererClass=&NiagaraMeshRendererProperties, rendererIndex=0}]
 }
}
summary {
 systemName=NS_Contrails_GhostChain_V1d9d1
 emitters [
  {emitterName=SwingChain, simTarget=GPUComputeSim, rendererClasses [&NiagaraMeshRendererProperties]}
  {emitterName=WarpChain, simTarget=GPUComputeSim, rendererClasses [&NiagaraMeshRendererProperties]}
  {emitterName=Flames-Local, rendererClasses [&NiagaraSpriteRendererProperties]}
  {emitterName=FireTrail, rendererClasses [&NiagaraMeshRendererProperties, &NiagaraLightRendererProperties]}
  {emitterName=Sparks, rendererClasses [&NiagaraSpriteRendererProperties]}
  {emitterName=Smoke_Wide_Large, rendererClasses [&NiagaraSpriteRendererProperties]}
  {emitterName=Empty, rendererClasses [&NiagaraSpriteRendererProperties]}
  {emitterName=Empty001, rendererClasses [&NiagaraSpriteRendererProperties]}
  {emitterName=SwingChain001, rendererClasses [&NiagaraMeshRendererProperties]}
  {emitterName=WarpChain001, rendererClasses [&NiagaraMeshRendererProperties]}
 ]
 userVariables [
  User.bIsFrontEnd {type=NiagaraBool}
  User.bIsFrontEndPreview {type=NiagaraBool}
  User.PlayerSpeed {type=NiagaraFloat, defaultValue=0, description=""}
  User.RibbonIDOffset {type=NiagaraInt32}
  User.SkeletalMesh {
   type=NiagaraDataInterfaceSkeletalMesh
   defaultValue {
    struct=NiagaraExt_VariableValue_DataInterface
    value {dataInterface=&self:NiagaraDataInterfaceSkeletalMesh_0}
   }
  }
 ]
}
